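# Base widget for FNE UI blocks, blending visual state settings into a user widget's color, translation, scale, shear and angle.
# Base, Selected and Locked layers each hold Idle, Focused, Hovered, Pressed and Disabled settings, combined through chained lerps weighted by Trans_* floats.
# Each layer feeds the next, with Selected and Locked gated by IsSelectable and IsLockable, and disabled entries get zero alpha.
# Helper functions snap Trans_* floats for instant state changes.
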WidgetBlueprint /FNE_UI/Systems/_Blocks/WBP_FNE_Block_Base
graphs [
 EventGraph {links [], nodes []}
 GetIsLocked {nodes [
   GetIsLocked() {is=title K2Node_FunctionEntry}
   > Return Node!(execute, IsLocked:bool=Get IsLocked() -> (IsLocked:bool))
  ]}
 GetIsSelectable {nodes [
   GetIsSelectable() {is=title K2Node_FunctionEntry}
   > Return Node!(execute, IsSelectable:bool=Get IsSelectable() -> (IsSelectable:bool))
  ]}
 GetIsSelected {nodes [
   GetIsSelected() {is=title K2Node_FunctionEntry}
   > Return Node!(execute, IsSelected:bool=Get IsSelected() -> (IsSelected:bool))
  ]}
 InstantTransitionToDisabled_BpFloats {nodes [
   n1 = InstantTransitionToDisabled_BpFloats!() -> (then, IsSelected:bool=~) {is=title K2Node_FunctionEntry}
   > n2 = Branch(Condition=@n1)
   > Set Trans_SelectedDisabled(Trans_SelectedDisabled:double=1) -> (Output_Get:double)
   Set Trans_Disabled(execute=@n2.else, Trans_Disabled:double=1) -> (Output_Get:double)
  ]}
 InstantTransitionToEnabled_BpFloats {nodes [
   n1 = InstantTransitionToEnabled_BpFloats!() -> (then, IsSelected:bool=~) {is=title K2Node_FunctionEntry}
   > n2 = Branch(Condition=@n1)
   Set Trans_SelectedHovered(execute=[@n2.then, @n3.then], Trans_SelectedHovered:double=0) -> (Output_Get:double)
   > Set Trans_SelectedFocused(Trans_SelectedFocused:double=0) -> (Output_Get:double)
   > Set Trans_SelectedPressed(Trans_SelectedPressed:double=0) -> (Output_Get:double)
   > Set Trans_SelectedDisabled(Trans_SelectedDisabled=0)
   Set Trans_Hovered(execute=@n2.else, Trans_Hovered:double=0) -> (Output_Get:double)
   > Set Trans_Focused(Trans_Focused:double=0) -> (Output_Get:double)
   > Set Trans_Pressed(Trans_Pressed:double=0) -> (Output_Get:double)
   > Set Trans_Disabled(Trans_Disabled=0)
   > n3 = Set Trans_SelectedIdle(Trans_SelectedIdle:double=0) -> (Output_Get:double)
  ]}
 InstantTransitionToSelected_BpFloats {nodes [
   InstantTransitionToSelected_BpFloats() {is=title K2Node_FunctionEntry}
   > Set Trans_SelectedIdle(Trans_SelectedIdle=1)
   > Set Trans_SelectedHovered(Trans_SelectedHovered=0)
   > Set Trans_SelectedFocused(Trans_SelectedFocused=0)
   > Set Trans_SelectedPressed(Trans_SelectedPressed=0)
   > Set Trans_SelectedDisabled(Trans_SelectedDisabled=0)
  ]}
 Sequencer Events {links [], nodes []}
 SetBaseStates {nodes [
   n1 = SetBaseStates!() -> (
    then
    InIdleSettings:S FNEUI States Settings Structure
    InFocusedSettings:S FNEUI States Settings Structure
    inHoveredSettings:S FNEUI States Settings Structure
    InPressedSettings:S FNEUI States Settings Structure
    InDisabledSettings:S FNEUI States Settings Structure
   ) {is=title K2Node_FunctionEntry}
   > Set IdleSettings(IdleSettings:S FNEUI States Settings Structure=@n1) -> (Output_Get:S FNEUI States Settings Structure)
   > Set FocusedSettings(FocusedSettings:S FNEUI States Settings Structure=@n1.InFocusedSettings) -> (Output_Get:S FNEUI States Settings Structure)
   > Set HoveredSettings(HoveredSettings:S FNEUI States Settings Structure=@n1.inHoveredSettings) -> (Output_Get:S FNEUI States Settings Structure)
   > Set PressedSettings(PressedSettings:S FNEUI States Settings Structure=@n1.InPressedSettings) -> (Output_Get:S FNEUI States Settings Structure)
   > Set DisabledSettings(DisabledSettings:S FNEUI States Settings Structure=@n1.InDisabledSettings) -> (Output_Get:S FNEUI States Settings Structure)
  ]}
 SetIsLockable {nodes [
   n1 = SetIsLockable!() -> (then, Is Lockable:bool=~) {is=title K2Node_FunctionEntry}
   > Set IsLockable(IsLockable:bool=@n1) -> (Output_Get:bool)
  ]}
 SetIsLocked {nodes [
   n1 = SetIsLocked!() -> (then, Is Locked:bool=~) {is=title K2Node_FunctionEntry}
   > Set IsLocked(IsLocked:bool=@n1) -> (Output_Get:bool)
  ]}
 SetIsSelectable {nodes [
   n1 = SetIsSelectable!() -> (then, Is Selectable:bool=~) {is=title K2Node_FunctionEntry}
   > Set IsSelectable(IsSelectable:bool=@n1) -> (Output_Get:bool)
  ]}
 SetIsSelected {nodes [
   n1 = SetIsSelected!() -> (then, IsSelected:bool=~) {is=title K2Node_FunctionEntry}
   > Set IsSelected(IsSelected:bool=@n1) -> (Output_Get:bool)
  ]}
 SetLockedStates {nodes [
   n1 = SetLockedStates!() -> (
    then
    InIdleSettings:S FNEUI States Settings Structure
    InFocusedSettings:S FNEUI States Settings Structure
    inHoveredSettings:S FNEUI States Settings Structure
    InPressedSettings:S FNEUI States Settings Structure
    InDisabledSettings:S FNEUI States Settings Structure
   ) {is=title K2Node_FunctionEntry}
   > Set LockedIdleSettings(LockedIdleSettings:S FNEUI States Settings Structure=@n1) -> (Output_Get:S FNEUI States Settings Structure)
   > Set LockedFocusedSettings(LockedFocusedSettings:S FNEUI States Settings Structure=@n1.InFocusedSettings) -> (Output_Get:S FNEUI States Settings Structure)
   > Set LockedHoveredSettings(LockedHoveredSettings:S FNEUI States Settings Structure=@n1.inHoveredSettings) -> (Output_Get:S FNEUI States Settings Structure)
   > Set LockedPressedSettings(LockedPressedSettings:S FNEUI States Settings Structure=@n1.InPressedSettings) -> (Output_Get:S FNEUI States Settings Structure)
   > Set LockedDisabledSettings(LockedDisabledSettings:S FNEUI States Settings Structure=@n1.InDisabledSettings) -> (Output_Get:S FNEUI States Settings Structure)
  ]}
 SetSelectedStates {nodes [
   n1 = SetSelectedStates!() -> (
    then
    InIdleSettings:S FNEUI States Settings Structure
    InFocusedSettings:S FNEUI States Settings Structure
    inHoveredSettings:S FNEUI States Settings Structure
    InPressedSettings:S FNEUI States Settings Structure
    InDisabledSettings:S FNEUI States Settings Structure
   ) {is=title K2Node_FunctionEntry}
   > Set SelectedIdleSettings(SelectedIdleSettings:S FNEUI States Settings Structure=@n1) -> (Output_Get:S FNEUI States Settings Structure)
   > Set SelectedFocusedSettings(SelectedFocusedSettings:S FNEUI States Settings Structure=@n1.InFocusedSettings) -> (Output_Get:S FNEUI States Settings Structure)
   > Set SelectedHoveredSettings(SelectedHoveredSettings:S FNEUI States Settings Structure=@n1.inHoveredSettings) -> (Output_Get:S FNEUI States Settings Structure)
   > Set SelectedPressedSettings(SelectedPressedSettings:S FNEUI States Settings Structure=@n1.InPressedSettings) -> (Output_Get:S FNEUI States Settings Structure)
   > Set SelectedDisabledSettings(SelectedDisabledSettings:S FNEUI States Settings Structure=@n1.InDisabledSettings) -> (Output_Get:S FNEUI States Settings Structure)
  ]}
 State Transitions Locked_Lerp_Vector2D {nodes [
   n1 = State Transitions Locked_Lerp_Vector2D!() -> (
    then
    InCurrentVector2D:Vector 2D Structure
    bIsIdleActive:bool=~
    LockedIdle:Vector 2D Structure
    bIsFocusedActive:bool=~
    LockedFocused:Vector 2D Structure
    bIsHoveredActive:bool=~
    LockedHovered:Vector 2D Structure
    bIsPressedActive:bool=~
    LockedPressed:Vector 2D Structure
    bIsDisabledActive:bool=~
    LockedDisabled:Vector 2D Structure
   ) {is=title K2Node_FunctionEntry}
   > Set LockedCurrentVector2D(LockedCurrentVector2D:Vector 2D Structure=@n1) -> (Output_Get:Vector 2D Structure)
   > n2 = Branch(Condition=Get IsLockable() -> (IsLockable:bool))
   > n3 = Set LockedCurrentVector2D()
   Return Node!(execute=[@n3.then, @n2.else], CurrentVector2D:Vector 2D Structure=Get LockedCurrentVector2D() -> (LockedCurrentVector2D:Vector 2D Structure))
   n4 = Get Trans_LockedIdle() -> (Trans_LockedIdle:double)
   n5 = Get bIsIdleActive() -> (bIsIdleActive:bool)
   n6 = FNE Lerp Vector 2D!(
    A:Vector 2D Structure=FNE Lerp Vector 2D(
     A=Get LockedCurrentVector2D()
     B=Get LockedIdle() -> (LockedIdle:Vector 2D Structure)
     Alpha=Select(Option 0:double=0, Option 1:double=@n4, Index=@n5) -> (ReturnValue:double)
    )
    B:Vector 2D Structure=Get LockedFocused() -> (LockedFocused:Vector 2D Structure)
    Alpha:double=Select(Option 0=0, Option 1=Get Trans_LockedFocused() -> (Trans_LockedFocused:double), Index=Get bIsFocusedActive() -> (bIsFocusedActive:bool))
   ) -> (ReturnValue:Vector 2D Structure)
   n7 = Select(Option 0=0, Option 1=Get Trans_LockedHovered() -> (Trans_LockedHovered:double), Index=Get bIsHoveredActive() -> (bIsHoveredActive:bool))
   n8 = Get LockedHovered() -> (LockedHovered:Vector 2D Structure)
   n9 = Get Trans_LockedPressed() -> (Trans_LockedPressed:double)
   n10 = Get bIsPressedActive() -> (bIsPressedActive:bool)
   FNE Lerp Vector 2D(
    A=FNE Lerp Vector 2D(
     A=FNE Lerp Vector 2D(A=@n6, B=@n8, Alpha=@n7)
     B=Get LockedPressed() -> (LockedPressed:Vector 2D Structure)
     Alpha=Select(Option 0=0, Option 1=@n9, Index=@n10)
    )
    B=Get LockedDisabled() -> (LockedDisabled:Vector 2D Structure)
    Alpha=Select(Option 0=0, Option 1=Get Trans_LockedDisabled() -> (Trans_LockedDisabled:double), Index=Get bIsDisabledActive() -> (bIsDisabledActive:bool))
   )
  ]}
 State TransitionsSelected_Lerp_Float {nodes [
   n1 = State TransitionsSelected_Lerp_Float!() -> (
    then
    InCurrentFLoat:double=~
    bIsIdleActive:bool=~
    SelectedIdle:double=~
    bIsFocusedActive:bool=~
    SelectedFocused:double=~
    bIsHoveredActive:bool=~
    SelectedHovered:double=~
    bIsPressedActive:bool=~
    SelectedPressed:double=~
    bIsDisabledActive:bool=~
    SelectedDisabled:double=~
   ) {is=title K2Node_FunctionEntry}
   > Set SelectedCurrentFloat(SelectedCurrentFloat:double=@n1) -> (Output_Get:double)
   > n2 = Branch(Condition=Get IsSelectable())
   n3 = Select(Option 0=0, Option 1=Get Trans_SelectedIdle() -> (Trans_SelectedIdle:double), Index=Get bIsIdleActive())
   n4 = Get SelectedCurrentFloat() -> (SelectedCurrentFloat:double)
   n5 = Get SelectedIdle() -> (SelectedIdle:double)
   n6 = Get bIsFocusedActive()
   n7 = Get Trans_SelectedFocused() -> (Trans_SelectedFocused:double)
   n8 = Lerp(
    A=Lerp(A=Lerp(A=@n4, B=@n5, Alpha=@n3), B=Get SelectedFocused() -> (SelectedFocused:double), Alpha=Select(Option 0=0, Option 1=@n7, Index=@n6))
    B=Get SelectedHovered() -> (SelectedHovered:double)
    Alpha=Select(Option 0=0, Option 1=Get Trans_SelectedHovered() -> (Trans_SelectedHovered:double), Index=Get bIsHoveredActive())
   )
   n9 = Select(Option 0=0, Option 1=Get Trans_SelectedPressed() -> (Trans_SelectedPressed:double), Index=Get bIsPressedActive())
   n10 = Get SelectedPressed() -> (SelectedPressed:double)
   n11 = Get bIsDisabledActive()
   n12 = Get Trans_SelectedDisabled() -> (Trans_SelectedDisabled:double)
   > n13 = Set SelectedCurrentFloat(SelectedCurrentFloat=Lerp(A=Lerp(A=@n8, B=@n10, Alpha=@n9), B=Get SelectedDisabled() -> (SelectedDisabled:double), Alpha=Select(Option 0=0, Option 1=@n12, Index=@n11)))
   Return Node!(execute=[@n13.then, @n2.else], CurrentFloat:double=Get SelectedCurrentFloat())
  ]}
 StateTransitionsBase_Lerp_Float {nodes [
   StateTransitionsBase_Lerp_Float!() -> (
    then
    Idle:double=~
    bIsFocusedActive:bool=~
    Focused:double=~
    bIsHoveredActive:bool=~
    Hovered:double=~
    bIsPressedActive:bool=~
    Pressed:double=~
    bIsDisabledActive:bool=~
    Disabled:double=~
   ) {is=title K2Node_FunctionEntry}
   n1 = Get Trans_Focused() -> (Trans_Focused:double)
   n2 = Get bIsFocusedActive()
   n3 = Lerp(
    A=Lerp(A=Get Idle() -> (Idle:double), B=Get Focused() -> (Focused:double), Alpha=Select(Option 0=0, Option 1=@n1, Index=@n2))
    B=Get Hovered() -> (Hovered:double)
    Alpha=Select(Option 0=0, Option 1=Get Trans_Hovered() -> (Trans_Hovered:double), Index=Get bIsHoveredActive())
   )
   n4 = Select(Option 0=0, Option 1=Get Trans_Pressed() -> (Trans_Pressed:double), Index=Get bIsPressedActive())
   n5 = Get Pressed() -> (Pressed:double)
   n6 = Get Trans_Disabled() -> (Trans_Disabled:double)
   n7 = Get bIsDisabledActive()
   > Return Node(CurrentFloat=Lerp(A=Lerp(A=@n3, B=@n5, Alpha=@n4), B=Get Disabled() -> (Disabled:double), Alpha=Select(Option 0=0, Option 1=@n6, Index=@n7)))
  ]}
 StateTransitionsBase_Lerp_LinearColor {nodes [
   StateTransitionsBase_Lerp_LinearColor!() -> (
    then
    Idle:Linear Color Structure
    bIsFocusedActive:bool=~
    Focused:Linear Color Structure
    bIsHoveredActive:bool=~
    Hovered:Linear Color Structure
    bIsPressedActive:bool=~
    Pressed:Linear Color Structure
    bIsDisabledActive:bool=~
    Disabled:Linear Color Structure
   ) {is=title K2Node_FunctionEntry}
   n1 = Get Trans_Focused()
   n2 = Get bIsFocusedActive()
   n3 = "Lerp (LinearColor)"(
    A="Lerp (LinearColor)"(
     A=Get Idle() -> (Idle:Linear Color Structure)
     B=Get Focused() -> (Focused:Linear Color Structure)
     Alpha=Select(Option 0=0, Option 1=@n1, Index=@n2)
    )
    B=Get Hovered() -> (Hovered:Linear Color Structure)
    Alpha=Select(Option 0=0, Option 1=Get Trans_Hovered(), Index=Get bIsHoveredActive())
   )
   n4 = Select(Option 0=0, Option 1=Get Trans_Pressed(), Index=Get bIsPressedActive())
   n5 = Get Pressed() -> (Pressed:Linear Color Structure)
   n6 = Get Trans_Disabled()
   n7 = Get bIsDisabledActive()
   > Return Node!(
    execute
    CurrentColor:Linear Color Structure="Lerp (LinearColor)"(
     A="Lerp (LinearColor)"(A=@n3, B=@n5, Alpha=@n4)
     B=Get Disabled() -> (Disabled:Linear Color Structure)
     Alpha=Select(Option 0=0, Option 1=@n6, Index=@n7)
    )
   )
  ]}
 StateTransitionsBase_Lerp_Vector2D {nodes [
   StateTransitionsBase_Lerp_Vector2D!() -> (
    then
    Idle:Vector 2D Structure
    bIsFocusedActive:bool=~
    Focused:Vector 2D Structure
    bIsHoveredActive:bool=~
    Hovered:Vector 2D Structure
    bIsPressedActive:bool=~
    Pressed:Vector 2D Structure
    bIsDisabledActive:bool=~
    Disabled:Vector 2D Structure
   ) {is=title K2Node_FunctionEntry}
   n1 = Get Trans_Focused()
   n2 = Get bIsFocusedActive()
   n3 = FNE Lerp Vector 2D(
    A=FNE Lerp Vector 2D(
     A=Get Idle() -> (Idle:Vector 2D Structure)
     B=Get Focused() -> (Focused:Vector 2D Structure)
     Alpha=Select(Option 0=0, Option 1=@n1, Index=@n2)
    )
    B=Get Hovered() -> (Hovered:Vector 2D Structure)
    Alpha=Select(Option 0=0, Option 1=Get Trans_Hovered(), Index=Get bIsHoveredActive())
   )
   n4 = Select(Option 0=0, Option 1=Get Trans_Pressed(), Index=Get bIsPressedActive())
   n5 = Get Pressed() -> (Pressed:Vector 2D Structure)
   n6 = Get Trans_Disabled()
   n7 = Get bIsDisabledActive()
   > Return Node!(
    execute
    CurrentVector2D:Vector 2D Structure=FNE Lerp Vector 2D(
     A=FNE Lerp Vector 2D(A=@n3, B=@n5, Alpha=@n4)
     B=Get Disabled() -> (Disabled:Vector 2D Structure)
     Alpha=Select(Option 0=0, Option 1=@n6, Index=@n7)
    )
   )
  ]}
 StateTransitionsLocked_Lerp_Float {nodes [
   n1 = StateTransitionsLocked_Lerp_Float!() -> (
    then
    InCurrentFloat:double=~
    bIsIdleActive:bool=~
    LockedIdle:double=~
    bIsFocusedActive:bool=~
    LockedFocused:double=~
    bIsHoveredActive:bool=~
    LockedHovered:double=~
    bIsPressedActive:bool=~
    LockedPressed:double=~
    bIsDisabledActive:bool=~
    LockedDisabled:double=~
   ) {is=title K2Node_FunctionEntry}
   > Set LockedCurrentFloat(LockedCurrentFloat:double=@n1) -> (Output_Get:double)
   > n2 = Branch(Condition=Get IsLockable())
   n3 = Select(Option 0=0, Option 1=Get Trans_LockedIdle(), Index=Get bIsIdleActive())
   n4 = Get LockedCurrentFloat() -> (LockedCurrentFloat:double)
   n5 = Get LockedIdle() -> (LockedIdle:double)
   n6 = Get Trans_LockedFocused()
   n7 = Get bIsFocusedActive()
   n8 = Lerp(
    A=Lerp(A=Lerp(A=@n4, B=@n5, Alpha=@n3), B=Get LockedFocused() -> (LockedFocused:double), Alpha=Select(Option 0=0, Option 1=@n6, Index=@n7))
    B=Get LockedHovered() -> (LockedHovered:double)
    Alpha=Select(Option 0=0, Option 1=Get Trans_LockedHovered(), Index=Get bIsHoveredActive())
   )
   n9 = Select(Option 0=0, Option 1=Get Trans_LockedPressed(), Index=Get bIsPressedActive())
   n10 = Get LockedPressed() -> (LockedPressed:double)
   n11 = Get bIsDisabledActive()
   n12 = Get Trans_LockedDisabled()
   > n13 = Set LockedCurrentFloat(LockedCurrentFloat=Lerp(A=Lerp(A=@n8, B=@n10, Alpha=@n9), B=Get LockedDisabled() -> (LockedDisabled:double), Alpha=Select(Option 0=0, Option 1=@n12, Index=@n11)))
   Return Node!(execute=[@n13.then, @n2.else], CurrentFloat:double=Get LockedCurrentFloat())
  ]}
 StateTransitionsLocked_Lerp_LinearColor {nodes [
   n1 = StateTransitionsLocked_Lerp_LinearColor!() -> (
    then
    InCurrentLinearColor:Linear Color Structure
    bIsIdleActive:bool=~
    LockedIdle:Linear Color Structure
    bIsFocusedActive:bool=~
    LockedFocused:Linear Color Structure
    bIsHoveredActive:bool=~
    LockedHovered:Linear Color Structure
    bIsPressedActive:bool=~
    LockedPressed:Linear Color Structure
    bIsDisabledActive:bool=~
    LockedDisabled:Linear Color Structure
   ) {is=title K2Node_FunctionEntry}
   > Set LockedCurrentLinearColor(LockedCurrentLinearColor:Linear Color Structure=@n1) -> (Output_Get:Linear Color Structure)
   > n2 = Branch(Condition=Get IsLockable())
   n3 = Select(Option 0=0, Option 1=Get Trans_LockedIdle(), Index=Get bIsIdleActive())
   n4 = Get LockedIdle() -> (LockedIdle:Linear Color Structure)
   n5 = Get LockedCurrentLinearColor() -> (LockedCurrentLinearColor:Linear Color Structure)
   n6 = Get Trans_LockedFocused()
   n7 = Get bIsFocusedActive()
   n8 = "Lerp (LinearColor)"(
    A="Lerp (LinearColor)"(
     A="Lerp (LinearColor)"(A=@n5, B=@n4, Alpha=@n3)
     B=Get LockedFocused() -> (LockedFocused:Linear Color Structure)
     Alpha=Select(Option 0=0, Option 1=@n6, Index=@n7)
    )
    B=Get LockedHovered() -> (LockedHovered:Linear Color Structure)
    Alpha=Select(Option 0=0, Option 1=Get Trans_LockedHovered(), Index=Get bIsHoveredActive())
   )
   n9 = Select(Option 0=0, Option 1=Get Trans_LockedPressed(), Index=Get bIsPressedActive())
   n10 = Get LockedPressed() -> (LockedPressed:Linear Color Structure)
   n11 = Get Trans_LockedDisabled()
   n12 = Get bIsDisabledActive()
   > n13 = Set LockedCurrentLinearColor(LockedCurrentLinearColor="Lerp (LinearColor)"(
     A="Lerp (LinearColor)"(A=@n8, B=@n10, Alpha=@n9)
     B=Get LockedDisabled() -> (LockedDisabled:Linear Color Structure)
     Alpha=Select(Option 0=0, Option 1=@n11, Index=@n12)
    ))
   Return Node!(execute=[@n13.then, @n2.else], CurrentColor:Linear Color Structure=Get LockedCurrentLinearColor())
  ]}
 StateTransitionsSelected_Lerp_LinearColor {nodes [
   n1 = StateTransitionsSelected_Lerp_LinearColor!() -> (
    then
    InCurrentLinearColor:Linear Color Structure
    bIsIdleActive:bool=~
    SelectedIdle:Linear Color Structure
    bIsFocusedActive:bool=~
    SelectedFocused:Linear Color Structure
    bIsHoveredActive:bool=~
    SelectedHovered:Linear Color Structure
    bIsPressedActive:bool=~
    SelectedPressed:Linear Color Structure
    bIsDisabledActive:bool=~
    SelectedDisabled:Linear Color Structure
   ) {is=title K2Node_FunctionEntry}
   > Set SelectedCurrentLinearColor(SelectedCurrentLinearColor:Linear Color Structure=@n1) -> (Output_Get:Linear Color Structure)
   > n2 = Branch(Condition=Get IsSelectable())
   n3 = Select(Option 0=0, Option 1=Get Trans_SelectedIdle(), Index=Get bIsIdleActive())
   n4 = Get SelectedIdle() -> (SelectedIdle:Linear Color Structure)
   n5 = Get SelectedCurrentLinearColor() -> (SelectedCurrentLinearColor:Linear Color Structure)
   n6 = Get bIsFocusedActive()
   n7 = Get Trans_SelectedFocused()
   n8 = "Lerp (LinearColor)"(
    A="Lerp (LinearColor)"(
     A="Lerp (LinearColor)"(A=@n5, B=@n4, Alpha=@n3)
     B=Get SelectedFocused() -> (SelectedFocused:Linear Color Structure)
     Alpha=Select(Option 0=0, Option 1=@n7, Index=@n6)
    )
    B=Get SelectedHovered() -> (SelectedHovered:Linear Color Structure)
    Alpha=Select(Option 0=0, Option 1=Get Trans_SelectedHovered(), Index=Get bIsHoveredActive())
   )
   n9 = Select(Option 0=0, Option 1=Get Trans_SelectedPressed(), Index=Get bIsPressedActive())
   n10 = Get SelectedPressed() -> (SelectedPressed:Linear Color Structure)
   n11 = Get bIsDisabledActive()
   n12 = Get Trans_SelectedDisabled()
   > n13 = Set SelectedCurrentLinearColor(SelectedCurrentLinearColor="Lerp (LinearColor)"(
     A="Lerp (LinearColor)"(A=@n8, B=@n10, Alpha=@n9)
     B=Get SelectedDisabled() -> (SelectedDisabled:Linear Color Structure)
     Alpha=Select(Option 0=0, Option 1=@n12, Index=@n11)
    ))
   Return Node(execute=[@n13.then, @n2.else], CurrentColor=Get SelectedCurrentLinearColor())
  ]}
 StateTransitionsSelected_Lerp_Vector2D {nodes [
   n1 = StateTransitionsSelected_Lerp_Vector2D!() -> (
    then
    InCurrentVector2D:Vector 2D Structure
    bIsIdleActive:bool=~
    SelectedIdle:Vector 2D Structure
    bIsFocusedActive:bool=~
    SelectedFocused:Vector 2D Structure
    bIsHoveredActive:bool=~
    SelectedHovered:Vector 2D Structure
    bIsPressedActive:bool=~
    SelectedPressed:Vector 2D Structure
    bIsDisabledActive:bool=~
    SelectedDisabled:Vector 2D Structure
   ) {is=title K2Node_FunctionEntry}
   > Set SelectedCurrentVector2D(SelectedCurrentVector2D:Vector 2D Structure=@n1) -> (Output_Get:Vector 2D Structure)
   > n2 = Branch(Condition=Get IsSelectable())
   n3 = Select(Option 0=0, Option 1=Get Trans_SelectedIdle(), Index=Get bIsIdleActive())
   n4 = Get SelectedCurrentVector2D() -> (SelectedCurrentVector2D:Vector 2D Structure)
   n5 = Get SelectedIdle() -> (SelectedIdle:Vector 2D Structure)
   n6 = Get bIsFocusedActive()
   n7 = Get Trans_SelectedFocused()
   n8 = FNE Lerp Vector 2D(
    A=FNE Lerp Vector 2D(
     A=FNE Lerp Vector 2D(A=@n4, B=@n5, Alpha=@n3)
     B=Get SelectedFocused() -> (SelectedFocused:Vector 2D Structure)
     Alpha=Select(Option 0=0, Option 1=@n7, Index=@n6)
    )
    B=Get SelectedHovered() -> (SelectedHovered:Vector 2D Structure)
    Alpha=Select(Option 0=0, Option 1=Get Trans_SelectedHovered(), Index=Get bIsHoveredActive())
   )
   n9 = Select(Option 0=0, Option 1=Get Trans_SelectedPressed(), Index=Get bIsPressedActive())
   n10 = Get SelectedPressed() -> (SelectedPressed:Vector 2D Structure)
   n11 = Get bIsDisabledActive()
   n12 = Get Trans_SelectedDisabled()
   > n13 = Set SelectedCurrentVector2D(SelectedCurrentVector2D=FNE Lerp Vector 2D(
     A=FNE Lerp Vector 2D(A=@n8, B=@n10, Alpha=@n9)
     B=Get SelectedDisabled() -> (SelectedDisabled:Vector 2D Structure)
     Alpha=Select(Option 0=0, Option 1=@n12, Index=@n11)
    ))
   Return Node!(execute=[@n13.then, @n2.else], CurrentVector2D:Vector 2D Structure=Get SelectedCurrentVector2D())
  ]}
 TransitionToDeselectedFocused_BpFloats {nodes [
   TransitionToDeselectedFocused_BpFloats() {is=title K2Node_FunctionEntry}
   > Set Trans_SelectedIdle(Trans_SelectedIdle=0)
   > Set Trans_SelectedHovered(Trans_SelectedHovered=0)
   > Set Trans_SelectedPressed(Trans_SelectedPressed=0)
   > Set Trans_SelectedDisabled(Trans_SelectedDisabled=0)
   > Set Trans_Pressed(Trans_Pressed=0)
   > Set Trans_Disabled(Trans_Disabled=0)
  ]}
 TransitionToDeselectedIdle_BpFloats {nodes [
   TransitionToDeselectedIdle_BpFloats() {is=title K2Node_FunctionEntry}
   > Set Trans_SelectedHovered(Trans_SelectedHovered=0)
   > Set Trans_SelectedFocused(Trans_SelectedFocused=0)
   > Set Trans_SelectedPressed(Trans_SelectedPressed=0)
   > Set Trans_SelectedDisabled(Trans_SelectedDisabled=0)
   > Set Trans_Hovered(Trans_Hovered=0)
   > Set Trans_Focused(Trans_Focused=0)
   > Set Trans_Pressed(Trans_Pressed=0)
   > Set Trans_Disabled(Trans_Disabled=0)
  ]}
 UpdateAnyFNE_BlockBaseInheritedWidgetCurrentStates {nodes [
   UpdateAnyFNE_BlockBaseInheritedWidgetCurrentStates!() -> (then, InCurrentUserWidget:User Widget Object Reference) {is=title K2Node_FunctionEntry}
   > n1 = Sequence!(execute) -> (then_0, then_1, then_2:Exec, then_3:Exec, then_4:Exec, then_5:Exec)
   > n2 = UpdateCurrentColor_Base!(execute, self) -> (then, CurrentColor:Linear Color Structure)
   > n3 = UpdateCurrentColor_Selected!(execute, self, InCurrentColor:Linear Color Structure=@n2) -> (then, CurrentColor:Linear Color Structure)
   > n4 = UpdateCurrentColor_Locked!(execute, self, InCurrentColor:Linear Color Structure=@n3) -> (then, CurrentColor:Linear Color Structure)
   > n5 = Set CurrentColor(CurrentColor:Linear Color Structure=@n4) -> (Output_Get:Linear Color Structure)
   > SetColorAndOpacity(
    self:User Widget Object Reference=Get InCurrentUserWidget() -> (InCurrentUserWidget:User Widget Object Reference)
    InColorAndOpacity:Linear Color Structure=@n5
   )
   n6 = UpdateCurrentTranslation_Base!(execute=@n1.then_1, self) -> (then, CurrentTranslation:Vector 2D Structure)
   > n7 = UpdateCurrentTranslation_Selected!(execute, self, InCurrentTranslation:Vector 2D Structure=@n6) -> (then, CurrentTranslation:Vector 2D Structure)
   > n8 = UpdateCurrentTranslation_Locked!(execute, self, InCurrentTranslation:Vector 2D Structure=@n7) -> (then, CurrentTranslation:Vector 2D Structure)
   > n9 = Set CurrentTranslation(CurrentTranslation:Vector 2D Structure=@n8) -> (Output_Get:Vector 2D Structure)
   > SetRenderTranslation(self=Get InCurrentUserWidget(), Translation=@n9)
   n10 = UpdateCurrentScale_Base!(execute=@n1.then_2, self) -> (then, CurrentScale:Vector 2D Structure)
   > n11 = UpdateCurrentScale_Selected!(execute, self, InCurrentScale:Vector 2D Structure=@n10) -> (then, CurrentScale:Vector 2D Structure)
   > n12 = UpdateCurrentScale_Locked!(execute, self, InCurrentScale:Vector 2D Structure=@n11) -> (then, CurrentScale:Vector 2D Structure)
   > n13 = Set CurrentScale(CurrentScale:Vector 2D Structure=@n12) -> (Output_Get:Vector 2D Structure)
   > SetRenderScale!(execute, self:Widget Object Reference=Get InCurrentUserWidget(), Scale:Vector 2D Structure=@n13) -> (then)
   n14 = UpdateCurrentShear_Base!(execute=@n1.then_3, self) -> (then, CurrentShear:Vector 2D Structure)
   > n15 = UpdateCurrentShear_Selected!(execute, self, InCurrentShear:Vector 2D Structure=@n14) -> (then, CurrentShear:Vector 2D Structure)
   > n16 = UpdateCurrentShear_Locked!(execute, self, InCurrentShear:Vector 2D Structure=@n15) -> (then, CurrentShear:Vector 2D Structure)
   > n17 = Set CurrentShear(CurrentShear:Vector 2D Structure=@n16) -> (Output_Get:Vector 2D Structure)
   > SetRenderShear!(execute, self:Widget Object Reference=Get InCurrentUserWidget(), Shear:Vector 2D Structure=@n17) -> (then)
   n18 = UpdateCurrentAngle_Base!(execute=@n1.then_4, self) -> (then, CurrentAngle:double)
   > n19 = UpdateCurrentAngle_Selected!(execute, self, InCurrentAngle:double=@n18) -> (then, CurrentAngle:double)
   > n20 = UpdateCurrentAngle_Locked!(execute, self, InCurrentAngle:double=@n19) -> (then, CurrentAngle:double)
   > n21 = Set CurrentAngle(CurrentAngle:double=@n20) -> (Output_Get:double)
   > SetRenderTransformAngle!(execute, self:Widget Object Reference=Get InCurrentUserWidget(), Angle:float=@n21) -> (then)
   Return Node!(
    execute=@n1.then_5
    CurrentColorOut:Linear Color Structure=Get CurrentColor() -> (CurrentColor:Linear Color Structure)
    CurrentTranslationOut:Vector 2D Structure=Get CurrentTranslation() -> (CurrentTranslation:Vector 2D Structure)
    CurrentScaleOut:Vector 2D Structure=Get CurrentScale() -> (CurrentScale:Vector 2D Structure)
    CurrentShearOut:Vector 2D Structure=Get CurrentShear() -> (CurrentShear:Vector 2D Structure)
    CurrentAngleOut:double=Get CurrentAngle() -> (CurrentAngle:double)
   )
  ]}
 UpdateCurrentAngle_Base {nodes [
   UpdateCurrentAngle_Base() {is=title K2Node_FunctionEntry}
   n1 = Break S FNEUI States Settings!(S_FNEUI_States_Settings:"S FNEUI States Settings Structure (by ref)"=Get HoveredSettings() -> (HoveredSettings:S FNEUI States Settings Structure)) -> (EnableState_50_#1:bool=~, Angle_46_#2:double=~) {is=title K2Node_BreakStruct}
   n2 = Break S FNEUI States Settings(S_FNEUI_States_Settings=Get FocusedSettings() -> (FocusedSettings:S FNEUI States Settings Structure)) -> (EnableState_50_#1=~, Angle_46_#2=~)
   n3 = Break S FNEUI States Settings(S_FNEUI_States_Settings=Get PressedSettings() -> (PressedSettings:S FNEUI States Settings Structure)) -> (EnableState_50_#1=~, Angle_46_#2=~)
   n4 = Break S FNEUI States Settings(S_FNEUI_States_Settings=Get DisabledSettings() -> (DisabledSettings:S FNEUI States Settings Structure)) -> (EnableState_50_#1=~, Angle_46_#2=~)
   > n5 = StateTransitionsBase_Lerp_Float!(
    execute
    self
    Idle:double=Break S FNEUI States Settings(S_FNEUI_States_Settings=Get IdleSettings() -> (IdleSettings:S FNEUI States Settings Structure)) -> (EnableState_50_#1=~, Angle_46_#2=~).Angle_46_#2
    bIsFocusedActive:bool=@n2
    Focused:double=@n2.Angle_46_#2
    bIsHoveredActive:bool=@n1
    Hovered:double=@n1.Angle_46_#2
    bIsPressedActive:bool=@n3
    Pressed:double=@n3.Angle_46_#2
    bIsDisabledActive:bool=@n4
    Disabled:double=@n4.Angle_46_#2
   ) -> (then, CurrentFloat:double) {is=title K2Node_CallFunction}
   > Return Node!(execute, CurrentAngle:double=@n5)
  ]}
 UpdateCurrentAngle_Locked {nodes [
   n1 = UpdateCurrentAngle_Locked!() -> (then, InCurrentAngle:double=~) {is=title K2Node_FunctionEntry}
   n2 = Break S FNEUI States Settings(S_FNEUI_States_Settings=Get LockedIdleSettings() -> (LockedIdleSettings:S FNEUI States Settings Structure)) -> (EnableState_50_#1=~, Angle_46_#2=~)
   n3 = Break S FNEUI States Settings(S_FNEUI_States_Settings=Get LockedHoveredSettings() -> (LockedHoveredSettings:S FNEUI States Settings Structure)) -> (EnableState_50_#1=~, Angle_46_#2=~)
   n4 = Break S FNEUI States Settings(S_FNEUI_States_Settings=Get LockedFocusedSettings() -> (LockedFocusedSettings:S FNEUI States Settings Structure)) -> (EnableState_50_#1=~, Angle_46_#2=~)
   n5 = Break S FNEUI States Settings(S_FNEUI_States_Settings=Get LockedPressedSettings() -> (LockedPressedSettings:S FNEUI States Settings Structure)) -> (EnableState_50_#1=~, Angle_46_#2=~)
   n6 = Break S FNEUI States Settings(S_FNEUI_States_Settings=Get LockedDisabledSettings() -> (LockedDisabledSettings:S FNEUI States Settings Structure)) -> (EnableState_50_#1=~, Angle_46_#2=~)
   > n7 = StateTransitionsLocked_Lerp_Float!(
    execute
    self
    InCurrentFloat:double=@n1
    bIsIdleActive:bool=@n2
    LockedIdle:double=@n2.Angle_46_#2
    bIsFocusedActive:bool=@n4
    LockedFocused:double=@n4.Angle_46_#2
    bIsHoveredActive:bool=@n3
    LockedHovered:double=@n3.Angle_46_#2
    bIsPressedActive:bool=@n5
    LockedPressed:double=@n5.Angle_46_#2
    bIsDisabledActive:bool=@n6
    LockedDisabled:double=@n6.Angle_46_#2
   ) -> (then, CurrentFloat:double) {is=title K2Node_CallFunction}
   > Return Node(CurrentAngle=@n7)
  ]}
 UpdateCurrentAngle_Selected {nodes [
   n1 = UpdateCurrentAngle_Selected!() -> (then, InCurrentAngle:double=~) {is=title K2Node_FunctionEntry}
   n2 = Break S FNEUI States Settings(S_FNEUI_States_Settings=Get SelectedDisabledSettings() -> (SelectedDisabledSettings:S FNEUI States Settings Structure)) -> (EnableState_50_#1=~, Angle_46_#2=~)
   n3 = Break S FNEUI States Settings(S_FNEUI_States_Settings=Get SelectedIdleSettings() -> (SelectedIdleSettings:S FNEUI States Settings Structure)) -> (EnableState_50_#1=~, Angle_46_#2=~)
   n4 = Break S FNEUI States Settings(S_FNEUI_States_Settings=Get SelectedHoveredSettings() -> (SelectedHoveredSettings:S FNEUI States Settings Structure)) -> (EnableState_50_#1=~, Angle_46_#2=~)
   n5 = Break S FNEUI States Settings(S_FNEUI_States_Settings=Get SelectedFocusedSettings() -> (SelectedFocusedSettings:S FNEUI States Settings Structure)) -> (EnableState_50_#1=~, Angle_46_#2=~)
   n6 = Break S FNEUI States Settings(S_FNEUI_States_Settings=Get SelectedPressedSettings() -> (SelectedPressedSettings:S FNEUI States Settings Structure)) -> (EnableState_50_#1=~, Angle_46_#2=~)
   > n7 = State TransitionsSelected_Lerp_Float!(
    execute
    self
    InCurrentFLoat:double=@n1
    bIsIdleActive:bool=@n3
    SelectedIdle:double=@n3.Angle_46_#2
    bIsFocusedActive:bool=@n5
    SelectedFocused:double=@n5.Angle_46_#2
    bIsHoveredActive:bool=@n4
    SelectedHovered:double=@n4.Angle_46_#2
    bIsPressedActive:bool=@n6
    SelectedPressed:double=@n6.Angle_46_#2
    bIsDisabledActive:bool=@n2
    SelectedDisabled:double=@n2.Angle_46_#2
   ) -> (then, CurrentFloat:double) {is=title K2Node_CallFunction}
   > Return Node(CurrentAngle=@n7)
  ]}
 UpdateCurrentColor_Base {nodes [
   UpdateCurrentColor_Base() {is=title K2Node_FunctionEntry}
   n1 = Break S FNEUI States Settings!(S_FNEUI_States_Settings=Get DisabledSettings()) -> (EnableState_50_#1=~, ColorandOpacity_45_#3:Linear Color Structure)
   n2 = Break S FNEUI States Settings(S_FNEUI_States_Settings=Get HoveredSettings()) -> (EnableState_50_#1=~)
   n3 = Break S FNEUI States Settings(S_FNEUI_States_Settings=Get FocusedSettings()) -> (EnableState_50_#1=~)
   n4 = Break S FNEUI States Settings(S_FNEUI_States_Settings=Get PressedSettings()) -> (EnableState_50_#1=~)
   > n5 = StateTransitionsBase_Lerp_LinearColor!(
    execute
    self
    Idle:Linear Color Structure=Break S FNEUI States Settings(S_FNEUI_States_Settings=Get IdleSettings()) -> (EnableState_50_#1=~).ColorandOpacity_45_#3
    bIsFocusedActive:bool=@n3
    Focused:Linear Color Structure=@n3.ColorandOpacity_45_#3
    bIsHoveredActive:bool=@n2
    Hovered:Linear Color Structure=@n2.ColorandOpacity_45_#3
    bIsPressedActive:bool=@n4
    Pressed:Linear Color Structure=@n4.ColorandOpacity_45_#3
    bIsDisabledActive:bool=@n1
    Disabled:Linear Color Structure=@n1.ColorandOpacity_45_#3
   ) -> (then, CurrentColor:Linear Color Structure) {is=title K2Node_CallFunction}
   > Return Node!(execute, CurrentColor:Linear Color Structure=@n5)
  ]}
 UpdateCurrentColor_Locked {nodes [
   n1 = UpdateCurrentColor_Locked!() -> (then, InCurrentColor:Linear Color Structure) {is=title K2Node_FunctionEntry}
   n2 = Break S FNEUI States Settings(S_FNEUI_States_Settings=Get LockedDisabledSettings()) -> (EnableState_50_#1=~)
   n3 = Break S FNEUI States Settings(S_FNEUI_States_Settings=Get LockedIdleSettings()) -> (EnableState_50_#1=~)
   n4 = Break S FNEUI States Settings(S_FNEUI_States_Settings=Get LockedHoveredSettings()) -> (EnableState_50_#1=~)
   n5 = Break S FNEUI States Settings(S_FNEUI_States_Settings=Get LockedFocusedSettings()) -> (EnableState_50_#1=~)
   n6 = Break S FNEUI States Settings(S_FNEUI_States_Settings=Get LockedPressedSettings()) -> (EnableState_50_#1=~)
   > n7 = StateTransitionsLocked_Lerp_LinearColor!(
    execute
    self
    InCurrentLinearColor:Linear Color Structure=@n1
    bIsIdleActive:bool=@n3
    LockedIdle:Linear Color Structure=@n3.ColorandOpacity_45_#3
    bIsFocusedActive:bool=@n5
    LockedFocused:Linear Color Structure=@n5.ColorandOpacity_45_#3
    bIsHoveredActive:bool=@n4
    LockedHovered:Linear Color Structure=@n4.ColorandOpacity_45_#3
    bIsPressedActive:bool=@n6
    LockedPressed:Linear Color Structure=@n6.ColorandOpacity_45_#3
    bIsDisabledActive:bool=@n2
    LockedDisabled:Linear Color Structure=@n2.ColorandOpacity_45_#3
   ) -> (then, CurrentColor:Linear Color Structure) {is=title K2Node_CallFunction}
   > Return Node(CurrentColor=@n7)
  ]}
 UpdateCurrentColor_Selected {nodes [
   n1 = UpdateCurrentColor_Selected!() -> (then, InCurrentColor:Linear Color Structure) {is=title K2Node_FunctionEntry}
   n2 = Break S FNEUI States Settings(S_FNEUI_States_Settings=Get SelectedDisabledSettings()) -> (EnableState_50_#1=~)
   n3 = Break S FNEUI States Settings(S_FNEUI_States_Settings=Get SelectedIdleSettings()) -> (EnableState_50_#1=~)
   n4 = Break S FNEUI States Settings(S_FNEUI_States_Settings=Get SelectedHoveredSettings()) -> (EnableState_50_#1=~)
   n5 = Break S FNEUI States Settings(S_FNEUI_States_Settings=Get SelectedFocusedSettings()) -> (EnableState_50_#1=~)
   n6 = Break S FNEUI States Settings(S_FNEUI_States_Settings=Get SelectedPressedSettings()) -> (EnableState_50_#1=~)
   > n7 = StateTransitionsSelected_Lerp_LinearColor!(
    execute
    self
    InCurrentLinearColor:Linear Color Structure=@n1
    bIsIdleActive:bool=@n3
    SelectedIdle:Linear Color Structure=@n3.ColorandOpacity_45_#3
    bIsFocusedActive:bool=@n5
    SelectedFocused:Linear Color Structure=@n5.ColorandOpacity_45_#3
    bIsHoveredActive:bool=@n4
    SelectedHovered:Linear Color Structure=@n4.ColorandOpacity_45_#3
    bIsPressedActive:bool=@n6
    SelectedPressed:Linear Color Structure=@n6.ColorandOpacity_45_#3
    bIsDisabledActive:bool=@n2
    SelectedDisabled:Linear Color Structure=@n2.ColorandOpacity_45_#3
   ) -> (then, CurrentColor:Linear Color Structure) {is=title K2Node_CallFunction}
   > Return Node(CurrentColor=@n7)
  ]}
 UpdateCurrentScale_Base {nodes [
   UpdateCurrentScale_Base() {is=title K2Node_FunctionEntry}
   n1 = Break S FNEUI States Settings!(S_FNEUI_States_Settings=Get DisabledSettings()) -> (EnableState_50_#1=~, Scale_34_#4:Vector 2D Structure)
   n2 = Break S FNEUI States Settings(S_FNEUI_States_Settings=Get HoveredSettings()) -> (EnableState_50_#1=~)
   n3 = Break S FNEUI States Settings(S_FNEUI_States_Settings=Get FocusedSettings()) -> (EnableState_50_#1=~)
   n4 = Break S FNEUI States Settings(S_FNEUI_States_Settings=Get PressedSettings()) -> (EnableState_50_#1=~)
   > n5 = StateTransitionsBase_Lerp_Vector2D!(
    execute
    self
    Idle:Vector 2D Structure=Break S FNEUI States Settings(S_FNEUI_States_Settings=Get IdleSettings()) -> (EnableState_50_#1=~).Scale_34_#4
    bIsFocusedActive:bool=@n3
    Focused:Vector 2D Structure=@n3.Scale_34_#4
    bIsHoveredActive:bool=@n2
    Hovered:Vector 2D Structure=@n2.Scale_34_#4
    bIsPressedActive:bool=@n4
    Pressed:Vector 2D Structure=@n4.Scale_34_#4
    bIsDisabledActive:bool=@n1
    Disabled:Vector 2D Structure=@n1.Scale_34_#4
   ) -> (then, CurrentVector2D:Vector 2D Structure) {is=title K2Node_CallFunction}
   > Return Node!(execute, CurrentScale:Vector 2D Structure=@n5)
  ]}
 UpdateCurrentScale_Locked {nodes [
   n1 = UpdateCurrentScale_Locked!() -> (then, InCurrentScale:Vector 2D Structure) {is=title K2Node_FunctionEntry}
   n2 = Break S FNEUI States Settings(S_FNEUI_States_Settings=Get LockedIdleSettings()) -> (EnableState_50_#1=~)
   n3 = Break S FNEUI States Settings(S_FNEUI_States_Settings=Get LockedHoveredSettings()) -> (EnableState_50_#1=~)
   n4 = Break S FNEUI States Settings(S_FNEUI_States_Settings=Get LockedFocusedSettings()) -> (EnableState_50_#1=~)
   n5 = Break S FNEUI States Settings(S_FNEUI_States_Settings=Get LockedPressedSettings()) -> (EnableState_50_#1=~)
   n6 = Break S FNEUI States Settings(S_FNEUI_States_Settings=Get LockedDisabledSettings()) -> (EnableState_50_#1=~)
   > n7 = State Transitions Locked_Lerp_Vector2D!(
    execute
    self
    InCurrentVector2D:Vector 2D Structure=@n1
    bIsIdleActive:bool=@n2
    LockedIdle:Vector 2D Structure=@n2.Scale_34_#4
    bIsFocusedActive:bool=@n4
    LockedFocused:Vector 2D Structure=@n4.Scale_34_#4
    bIsHoveredActive:bool=@n3
    LockedHovered:Vector 2D Structure=@n3.Scale_34_#4
    bIsPressedActive:bool=@n5
    LockedPressed:Vector 2D Structure=@n5.Scale_34_#4
    bIsDisabledActive:bool=@n6
    LockedDisabled:Vector 2D Structure=@n6.Scale_34_#4
   ) -> (then, CurrentVector2D:Vector 2D Structure) {is=title K2Node_CallFunction}
   > Return Node(CurrentScale=@n7)
  ]}
 UpdateCurrentScale_Selected {nodes [
   n1 = UpdateCurrentScale_Selected!() -> (then, InCurrentScale:Vector 2D Structure) {is=title K2Node_FunctionEntry}
   n2 = Break S FNEUI States Settings(S_FNEUI_States_Settings=Get SelectedDisabledSettings()) -> (EnableState_50_#1=~)
   n3 = Break S FNEUI States Settings(S_FNEUI_States_Settings=Get SelectedIdleSettings()) -> (EnableState_50_#1=~)
   n4 = Break S FNEUI States Settings(S_FNEUI_States_Settings=Get SelectedHoveredSettings()) -> (EnableState_50_#1=~)
   n5 = Break S FNEUI States Settings(S_FNEUI_States_Settings=Get SelectedFocusedSettings()) -> (EnableState_50_#1=~)
   n6 = Break S FNEUI States Settings(S_FNEUI_States_Settings=Get SelectedPressedSettings()) -> (EnableState_50_#1=~)
   > n7 = StateTransitionsSelected_Lerp_Vector2D!(
    execute
    self
    InCurrentVector2D:Vector 2D Structure=@n1
    bIsIdleActive:bool=@n3
    SelectedIdle:Vector 2D Structure=@n3.Scale_34_#4
    bIsFocusedActive:bool=@n5
    SelectedFocused:Vector 2D Structure=@n5.Scale_34_#4
    bIsHoveredActive:bool=@n4
    SelectedHovered:Vector 2D Structure=@n4.Scale_34_#4
    bIsPressedActive:bool=@n6
    SelectedPressed:Vector 2D Structure=@n6.Scale_34_#4
    bIsDisabledActive:bool=@n2
    SelectedDisabled:Vector 2D Structure=@n2.Scale_34_#4
   ) -> (then, CurrentVector2D:Vector 2D Structure) {is=title K2Node_CallFunction}
   > Return Node(CurrentScale=@n7)
  ]}
 UpdateCurrentShear_Base {nodes [
   UpdateCurrentShear_Base() {is=title K2Node_FunctionEntry}
   n1 = Break S FNEUI States Settings!(S_FNEUI_States_Settings=Get DisabledSettings()) -> (EnableState_50_#1=~, Shear_36_#5:Vector 2D Structure)
   n2 = Break S FNEUI States Settings(S_FNEUI_States_Settings=Get HoveredSettings()) -> (EnableState_50_#1=~)
   n3 = Break S FNEUI States Settings(S_FNEUI_States_Settings=Get FocusedSettings()) -> (EnableState_50_#1=~)
   n4 = Break S FNEUI States Settings(S_FNEUI_States_Settings=Get PressedSettings()) -> (EnableState_50_#1=~)
   > n5 = StateTransitionsBase_Lerp_Vector2D(
    Idle=Break S FNEUI States Settings(S_FNEUI_States_Settings=Get IdleSettings()) -> (EnableState_50_#1=~).Shear_36_#5
    bIsFocusedActive=@n3
    Focused=@n3.Shear_36_#5
    bIsHoveredActive=@n2
    Hovered=@n2.Shear_36_#5
    bIsPressedActive=@n4
    Pressed=@n4.Shear_36_#5
    bIsDisabledActive=@n1
    Disabled=@n1.Shear_36_#5
   )
   > Return Node!(execute, CurrentShear:Vector 2D Structure=@n5)
  ]}
 UpdateCurrentShear_Locked {nodes [
   n1 = UpdateCurrentShear_Locked!() -> (then, InCurrentShear:Vector 2D Structure) {is=title K2Node_FunctionEntry}
   n2 = Break S FNEUI States Settings(S_FNEUI_States_Settings=Get LockedIdleSettings()) -> (EnableState_50_#1=~)
   n3 = Break S FNEUI States Settings(S_FNEUI_States_Settings=Get LockedHoveredSettings()) -> (EnableState_50_#1=~)
   n4 = Break S FNEUI States Settings(S_FNEUI_States_Settings=Get LockedFocusedSettings()) -> (EnableState_50_#1=~)
   n5 = Break S FNEUI States Settings(S_FNEUI_States_Settings=Get LockedPressedSettings()) -> (EnableState_50_#1=~)
   n6 = Break S FNEUI States Settings(S_FNEUI_States_Settings=Get LockedDisabledSettings()) -> (EnableState_50_#1=~)
   > n7 = State Transitions Locked_Lerp_Vector2D(
    InCurrentVector2D=@n1
    bIsIdleActive=@n2
    LockedIdle=@n2.Shear_36_#5
    bIsFocusedActive=@n4
    LockedFocused=@n4.Shear_36_#5
    bIsHoveredActive=@n3
    LockedHovered=@n3.Shear_36_#5
    bIsPressedActive=@n5
    LockedPressed=@n5.Shear_36_#5
    bIsDisabledActive=@n6
    LockedDisabled=@n6.Shear_36_#5
   )
   > Return Node(CurrentShear=@n7)
  ]}
 UpdateCurrentShear_Selected {nodes [
   n1 = UpdateCurrentShear_Selected!() -> (then, InCurrentShear:Vector 2D Structure) {is=title K2Node_FunctionEntry}
   n2 = Break S FNEUI States Settings(S_FNEUI_States_Settings=Get SelectedDisabledSettings()) -> (EnableState_50_#1=~)
   n3 = Break S FNEUI States Settings(S_FNEUI_States_Settings=Get SelectedIdleSettings()) -> (EnableState_50_#1=~)
   n4 = Break S FNEUI States Settings(S_FNEUI_States_Settings=Get SelectedHoveredSettings()) -> (EnableState_50_#1=~)
   n5 = Break S FNEUI States Settings(S_FNEUI_States_Settings=Get SelectedFocusedSettings()) -> (EnableState_50_#1=~)
   n6 = Break S FNEUI States Settings(S_FNEUI_States_Settings=Get SelectedPressedSettings()) -> (EnableState_50_#1=~)
   > n7 = StateTransitionsSelected_Lerp_Vector2D(
    InCurrentVector2D=@n1
    bIsIdleActive=@n3
    SelectedIdle=@n3.Shear_36_#5
    bIsFocusedActive=@n5
    SelectedFocused=@n5.Shear_36_#5
    bIsHoveredActive=@n4
    SelectedHovered=@n4.Shear_36_#5
    bIsPressedActive=@n6
    SelectedPressed=@n6.Shear_36_#5
    bIsDisabledActive=@n2
    SelectedDisabled=@n2.Shear_36_#5
   )
   > Return Node(CurrentShear=@n7)
  ]}
 UpdateCurrentTranslation_Base {nodes [
   UpdateCurrentTranslation_Base() {is=title K2Node_FunctionEntry}
   n1 = Break S FNEUI States Settings!(S_FNEUI_States_Settings=Get DisabledSettings()) -> (EnableState_50_#1=~, Translation_32_#6:Vector 2D Structure)
   n2 = Break S FNEUI States Settings(S_FNEUI_States_Settings=Get HoveredSettings()) -> (EnableState_50_#1=~)
   n3 = Break S FNEUI States Settings(S_FNEUI_States_Settings=Get FocusedSettings()) -> (EnableState_50_#1=~)
   n4 = Break S FNEUI States Settings(S_FNEUI_States_Settings=Get PressedSettings()) -> (EnableState_50_#1=~)
   > n5 = StateTransitionsBase_Lerp_Vector2D(
    Idle=Break S FNEUI States Settings(S_FNEUI_States_Settings=Get IdleSettings()) -> (EnableState_50_#1=~).Translation_32_#6
    bIsFocusedActive=@n3
    Focused=@n3.Translation_32_#6
    bIsHoveredActive=@n2
    Hovered=@n2.Translation_32_#6
    bIsPressedActive=@n4
    Pressed=@n4.Translation_32_#6
    bIsDisabledActive=@n1
    Disabled=@n1.Translation_32_#6
   )
   > Return Node!(execute, CurrentTranslation:Vector 2D Structure=@n5)
  ]}
 UpdateCurrentTranslation_Locked {nodes [
   n1 = UpdateCurrentTranslation_Locked!() -> (then, InCurrentTranslation:Vector 2D Structure) {is=title K2Node_FunctionEntry}
   n2 = Break S FNEUI States Settings(S_FNEUI_States_Settings=Get LockedIdleSettings()) -> (EnableState_50_#1=~)
   n3 = Break S FNEUI States Settings(S_FNEUI_States_Settings=Get LockedHoveredSettings()) -> (EnableState_50_#1=~)
   n4 = Break S FNEUI States Settings(S_FNEUI_States_Settings=Get LockedFocusedSettings()) -> (EnableState_50_#1=~)
   n5 = Break S FNEUI States Settings(S_FNEUI_States_Settings=Get LockedPressedSettings()) -> (EnableState_50_#1=~)
   n6 = Break S FNEUI States Settings(S_FNEUI_States_Settings=Get LockedDisabledSettings()) -> (EnableState_50_#1=~)
   > n7 = State Transitions Locked_Lerp_Vector2D(
    InCurrentVector2D=@n1
    bIsIdleActive=@n2
    LockedIdle=@n2.Translation_32_#6
    bIsFocusedActive=@n4
    LockedFocused=@n4.Translation_32_#6
    bIsHoveredActive=@n3
    LockedHovered=@n3.Translation_32_#6
    bIsPressedActive=@n5
    LockedPressed=@n5.Translation_32_#6
    bIsDisabledActive=@n6
    LockedDisabled=@n6.Translation_32_#6
   )
   > Return Node(CurrentTranslation=@n7)
  ]}
 UpdateCurrentTranslation_Selected {nodes [
   n1 = UpdateCurrentTranslation_Selected!() -> (then, InCurrentTranslation:Vector 2D Structure) {is=title K2Node_FunctionEntry}
   n2 = Break S FNEUI States Settings(S_FNEUI_States_Settings=Get SelectedDisabledSettings()) -> (EnableState_50_#1=~)
   n3 = Break S FNEUI States Settings(S_FNEUI_States_Settings=Get SelectedIdleSettings()) -> (EnableState_50_#1=~)
   n4 = Break S FNEUI States Settings(S_FNEUI_States_Settings=Get SelectedHoveredSettings()) -> (EnableState_50_#1=~)
   n5 = Break S FNEUI States Settings(S_FNEUI_States_Settings=Get SelectedFocusedSettings()) -> (EnableState_50_#1=~)
   n6 = Break S FNEUI States Settings(S_FNEUI_States_Settings=Get SelectedPressedSettings()) -> (EnableState_50_#1=~)
   > n7 = StateTransitionsSelected_Lerp_Vector2D(
    InCurrentVector2D=@n1
    bIsIdleActive=@n3
    SelectedIdle=@n3.Translation_32_#6
    bIsFocusedActive=@n5
    SelectedFocused=@n5.Translation_32_#6
    bIsHoveredActive=@n4
    SelectedHovered=@n4.Translation_32_#6
    bIsPressedActive=@n6
    SelectedPressed=@n6.Translation_32_#6
    bIsDisabledActive=@n2
    SelectedDisabled=@n2.Translation_32_#6
   )
   > Return Node(CurrentTranslation=@n7)
  ]}
]
variables [
 DisabledSettings {
  type=S_FNEUI_States_Settings
  default {angle=0, enable State=false, color and Opacity=<rgba 1 1 1 1>, scale=<1 1>, shear=<0 0>, translation=<0 0>}
 }
 FocusedSettings {
  type=S_FNEUI_States_Settings
  default {angle=0, enable State=false, color and Opacity=<rgba 1 1 1 1>, scale=<1 1>, shear=<0 0>, translation=<0 0>}
 }
 HoveredSettings {
  type=S_FNEUI_States_Settings
  default {angle=0, enable State=false, color and Opacity=<rgba 1 1 1 1>, scale=<1 1>, shear=<0 0>, translation=<0 0>}
 }
 IdleSettings {
  type=S_FNEUI_States_Settings
  default {angle=0, enable State=false, color and Opacity=<rgba 1 1 1 1>, scale=<1 1>, shear=<0 0>, translation=<0 0>}
 }
 IsLockable {type=boolean}
 IsLocked {type=boolean}
 IsSelectable {type=boolean}
 IsSelected {type=boolean}
 LockedDisabledSettings {
  type=S_FNEUI_States_Settings
  default {angle=0, enable State=false, color and Opacity=<rgba 1 1 1 1>, scale=<1 1>, shear=<0 0>, translation=<0 0>}
 }
 LockedFocusedSettings {
  type=S_FNEUI_States_Settings
  default {angle=0, enable State=false, color and Opacity=<rgba 1 1 1 1>, scale=<1 1>, shear=<0 0>, translation=<0 0>}
 }
 LockedHoveredSettings {
  type=S_FNEUI_States_Settings
  default {angle=0, enable State=false, color and Opacity=<rgba 1 1 1 1>, scale=<1 1>, shear=<0 0>, translation=<0 0>}
 }
 LockedIdleSettings {
  type=S_FNEUI_States_Settings
  default {angle=0, enable State=false, color and Opacity=<rgba 1 1 1 1>, scale=<1 1>, shear=<0 0>, translation=<0 0>}
 }
 LockedPressedSettings {
  type=S_FNEUI_States_Settings
  default {angle=0, enable State=false, color and Opacity=<rgba 1 1 1 1>, scale=<1 1>, shear=<0 0>, translation=<0 0>}
 }
 PressedSettings {
  type=S_FNEUI_States_Settings
  default {angle=0, enable State=false, color and Opacity=<rgba 1 1 1 1>, scale=<1 1>, shear=<0 0>, translation=<0 0>}
 }
 SelectedDisabledSettings {
  type=S_FNEUI_States_Settings
  default {angle=0, enable State=false, color and Opacity=<rgba 1 1 1 1>, scale=<1 1>, shear=<0 0>, translation=<0 0>}
 }
 SelectedFocusedSettings {
  type=S_FNEUI_States_Settings
  default {angle=0, enable State=false, color and Opacity=<rgba 1 1 1 1>, scale=<1 1>, shear=<0 0>, translation=<0 0>}
 }
 SelectedHoveredSettings {
  type=S_FNEUI_States_Settings
  default {angle=0, enable State=false, color and Opacity=<rgba 1 1 1 1>, scale=<1 1>, shear=<0 0>, translation=<0 0>}
 }
 SelectedIdleSettings {
  type=S_FNEUI_States_Settings
  default {angle=0, enable State=false, color and Opacity=<rgba 1 1 1 1>, scale=<1 1>, shear=<0 0>, translation=<0 0>}
 }
 SelectedPressedSettings {
  type=S_FNEUI_States_Settings
  default {angle=0, enable State=false, color and Opacity=<rgba 1 1 1 1>, scale=<1 1>, shear=<0 0>, translation=<0 0>}
 }
 Trans_Disabled {type=number}
 Trans_Focused {type=number}
 Trans_Hovered {type=number}
 Trans_LockedDisabled {type=number}
 Trans_LockedFocused {type=number}
 Trans_LockedHovered {type=number}
 Trans_LockedIdle {type=number}
 Trans_LockedPressed {type=number}
 Trans_Pressed {type=number}
 Trans_SelectedDisabled {type=number}
 Trans_SelectedFocused {type=number}
 Trans_SelectedHovered {type=number}
 Trans_SelectedIdle {type=number}
 Trans_SelectedPressed {type=number}
]
widgets {info {inheritedWidgetCount=0, parentClass=/Script/FNE_UILibrary.FNE_UIBlock, rootWidgetClass=null, widgetCount=0}, widgets []}
